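# Niagara effect for the ClayPlug Graffiti Jetwheels emote, covering both rolling wheels and hovering jets on a skeletal mesh.
# User.JetActiveCurve crossfades kicked-up chunks and wheel-spin meshes against four wheel thrusters with lights, three rear thrusters, windstreaks and ground rings.
# Bone and socket sampling places wheel and thruster effects, while smoke shifts with jet state.
# Owner velocity, water state, front-end overrides and time of day adjust spawning, scale and color.
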NiagaraSystem /BRCosmetics/Animation/Game/MainPlayer/Emotes/ClayPlug_Graffiti/FX/NS_Clayplug_Emote_Jetwheels
dependencies {
 usedDataInterfaces [
  &NiagaraDataInterfaceArrayFloat3
  &NiagaraDataInterfaceArrayInt32
  &NiagaraDataInterfaceSpriteRendererInfo
  &NiagaraDataInterfaceVector2DCurve
  &NiagaraDataInterfaceCurve
  &NiagaraDataInterfaceArrayFloat
  &NiagaraDataInterfaceColorCurve
  &NiagaraDataInterfaceSkeletalMesh
 ]
 usedDynamicInputs [
  &MakeFloatFromVector2d
  &Multiply_Vector2
  &RandomRangeVector
  &SimulationPosition
 ]
 usedModules [
  &self:SystemSpawnScript.SystemScriptSource.SystemScriptGraph.NiagaraNodeAssignment_0.SetVariables_#1
  &SystemState
  &Timeline
  /BRCosmetics/Characters/Player/Female/Medium/Bodies/F_MED_PiperShelf_Pearl/FX/TimeOfDayGlobal
 ]
}
emitters {
 Chunks {
  emitterUpdateScript [
   &EmitterState()
   &SpawnRate(
    SpawnRate=&FN_FrontEndFloatModifier(
     Gameplay=0
     Front End=&Multiply_Float(
      A=&Lerp_Float(A=&RandomRangeFloat(Minimum=2, Maximum=5), B=0, Alpha=$User.JetActiveCurve)
      B=&FloatFromCurve(
       FloatCurve=DataInterface() {propertyValues {Curve {keys [{tangentMode=RCTM_User, time=0, value=0}, {tangentMode=RCTM_User, time=1, value=1}]}}}
       CurveIndex=$Engine.System.Age
      )
     )
    )
    Use Spawn Probability=false
   )
   &SpawnPerUnit(
    Module.Use Max Movement Threshold=true
    Module.Use Movement Tolerance=true
    Movement Tolerance=2
    Spawn Probability:NiagaraFloat=&MakeCustomFloatFromBool(
     Bool=&SetBoolByFloatComparison(A=$User.JetActiveCurve, B=0.1)
     True Float=0
     False Float=&Multiply_Float(A=0.5, B=$User.DustSpawnCurve)
    )
    Spawn Spacing:NiagaraFloat=&Multiply_Float(A=&MakeCustomFloatFromBool(Bool=$User.bIsInWater, True Float=0, False Float=1), B=5)
    Use Spawn Probability=true
   ) {module=SpawnPerUnit001}
  ]
  particleSpawnScript [
   &PartitionParticles(Number Of Partitions=4)
   &InitializeParticle(
    Color=&ColorFromCurve(
     ColorCurve=DataInterface() {propertyValues {
       AlphaCurve {keys [<at -0.0023 1 RCIM_Cubic>]}
       BlueCurve {keys [<at -0.0023 0.2122 RCIM_Cubic>,<at 0.1383 0.2122>,<at 0.2178 0.1859>,<at 0.4486 0.75>,<at 0.7126 0.1859>,<at 1.0234 0.8802>]}
       GreenCurve {keys [<at -0.0023 0.2325 RCIM_Cubic>,<at 0.1383 0.2325>,<at 0.2178 0.2005>,<at 0.4486 0.8456>,<at 0.7126 0.2122>,<at 1.0234 0.9474>]}
       RedCurve {keys [<at -0.0023 0.25 RCIM_Cubic>,<at 0.1383 0.25>,<at 0.2178 0.2188>,<at 0.4486 1>,<at 0.7126 0.2188>,<at 1.0234 1>]}
      }}
     CurveIndex=&RandomRangeFloat(Minimum=0, Maximum=1)
    )
    Lifetime Max=1
    Lifetime Min=0.25
    Lifetime Mode=::NewEnumerator1
    Sprite Rotation Mode=::NewEnumerator2
    Sprite Size Mode=::NewEnumerator1
    Uniform Sprite Size Max=6
    Uniform Sprite Size Min=4
    UsePositionOffset=false
    Write Material Random=true
   ) {module=InitializeParticle001}
   &SkeletalMeshLocation(
    Bone / Socket Index=$Particles.PartitionParticles.Partition
    Bone Sampling Mode=::NewEnumerator4
    MaskSpawnGroup=false
    Skeletal Mesh=$System.Skeletal Mesh
   )
   /SetVariables(Particles.Position=&AddVectorToPosition(
     Position=$Particles.SkeletalMeshLocation.SampledPosition
     Vector=&RotateVectorByQuaternion(Vector To Rotate=<0 0 -20>, Quaternion=$Engine.Owner.Rotation)
    )) {isSetParameters=true, module=SetVariables_#2}
   &ShapeLocation(MaskSpawnGroup=false, Override Seed=false, Sphere Radius=5, UseOwnerScale=true)
   &AddVelocity(Velocity=<0 -25 10>) {module=AddVelocity001}
   &AddVelocityInCone(
    Cone Angle=25
    Cone Axis=<0 -2 1>
    Module.Use Velocity Falloff On Cone Axis=false
    Velocity Strength=&UniformRangedFloat(Minimum=50, Maximum=300)
   )
  ]
  particleUpdateScript [
   &ParticleState()
   &Drag(Use Linear Drag=true, Use Rotational Drag=true)
   &GravityForce(Gravity=<0 0 -500>)
   &SolveForcesAndVelocity(Clamp Velocity=false, Limit Acceleration=false)
   &ScaleColor(Scale RGB=<1.2 1.2 1.2>, ScaleA=true, ScaleRGB=true)
   &ScaleSpriteSize(Uniform Curve Sprite Scale=DataInterface() {propertyValues {Curve {keys [
        <at 0.7 0.9701>
        {arriveTangent=-3.2149, interpMode=RCIM_Cubic, leaveTangent=-3.2149, tangentMode=RCTM_User, time=0.9205, value=0.9191}
        <at 1.0011 0.002>
       ]}}})
   &SpriteRotationRate()
   &ScaleSpriteSizeBySpeed(
    Max Scale Factor=&Vector2DFromFloat(Value=1)
    Min Scale Factor=&Vector2DFromFloat(Value=0.25)
    Velocity Threshold=200
   )
  ]
  rendererClasses [&NiagaraSpriteRendererProperties]
  renderers [{rendererClass=&NiagaraSpriteRendererProperties}]
 }
 Default_Emote_TraversalSmoke_Loop001 {
  emitterSpawnScript [/SetVariables(Emitter.FakeLightActive=true) {isSetParameters=true}]
  emitterUpdateScript [
   &EmitterState()
   &SpawnRate(
    SpawnRate=&FN_FrontEndFloatModifier(
     Gameplay=&MakeCustomFloatFromBool(Bool=&SetBoolByFloatComparison(A=$User.JetActiveCurve, B=0.1), True Float=30, False Float=0)
     Front End=&Multiply_Float(
      A=44
      B=&Multiply_Float(
       A=$User.DustSpawnCurve
       B=&FloatFromCurve(
        FloatCurve=DataInterface() {propertyValues {Curve {keys [{tangentMode=RCTM_User, time=0.5, value=0}, {tangentMode=RCTM_User, time=1, value=1}]}}}
        CurveIndex=$Engine.System.Age
       )
      )
     )
    )
    Use Spawn Probability=false
   )
   &SpawnPerUnit(
    Module.Use Max Movement Threshold=true
    Module.Use Movement Tolerance=true
    Movement Tolerance=2
    Spawn Probability:NiagaraFloat=$User.DustSpawnCurve
    Spawn Spacing=5
    Use Spawn Probability=true
   ) {module=SpawnPerUnit001}
  ]
  particleSpawnScript [
   &PartitionParticles(Number Of Partitions=4)
   &InitializeParticle(
    Color Channel Mode:ENiagaraChannelCorrelation=::NewEnumerator0
    Color Maximum=<rgba 0.35 0.3477 0.336 0.25>
    Color Minimum=<rgba 0.35 0.3404 0.3251 0.2>
    Color Mode=::NewEnumerator2
    Lifetime Max=0.8
    Lifetime Min=0.5
    Lifetime Mode=::NewEnumerator1
    Sprite Rotation Mode=::NewEnumerator2
    Sprite Size Mode=::NewEnumerator1
    Uniform Sprite Size Max=25
    Uniform Sprite Size Min=15
    UsePositionOffset=false
    Write Material Random=true
   ) {module=InitializeParticle001}
   &ShapeLocation(
    Cylinder Height=1
    Cylinder Radius=&MakeCustomFloatFromBool(Bool=&SetBoolByFloatComparison(A=$User.JetActiveCurve, B=0.5), True Float=35, False Float=0)
    MaskSpawnGroup=false
    Offset=&SelectVectorFromArray(
     Array Sampling Mode=::NewEnumerator0
     Direct Array Index=?
     Vector Selection Array=DataInterface() {propertyValues {FloatData [<50 -40 0>, <-50 -40 0>, <50 60 0>, <-50 60 0>]}}
     Random Seed=0
    )
    Offset Coordinate Space=::Local
    Offset Mode=::NewEnumerator0
    Override Seed=false
    Shape Primitive=::NewEnumerator1
    Surface Only=false
    UseOwnerScale=true
   )
   &SkeletalMeshLocation(
    Bone / Socket Index=$Particles.PartitionParticles.Partition
    Bone Sampling Mode=::NewEnumerator4
    MaskSpawnGroup=false
    Skeletal Mesh=$System.Skeletal Mesh
   )
   /SetVariables(Particles.Position=&MakeCustomVectorFromBool(
     Bool=&SetBoolByFloatComparison(A=$System.Timeline.Float, B=0.1)
     True Vector=$Particles.ShapeLocation.ShapePosition
     False Vector=&AddVectorToPosition(
      Position=$Particles.SkeletalMeshLocation.SampledPosition
      Vector=&RotateVectorByQuaternion(Vector To Rotate=<0 0 -10>, Quaternion=$Engine.Owner.Rotation)
     )
    )) {isSetParameters=true, module=SetVariables_#2}
   &InheritVelocity(Inherited Velocity Speed Limit=1000, Speed Limit=true, UseSpeedThreshold=false) {module=InheritVelocity001}
   &AddVelocity(Velocity=<0 -2 1>, Velocity Speed Scale=&RandomRangeFloat(Minimum=0, Maximum=100)) {module=AddVelocity001}
   &AddVelocity(
    Module.Use Falloff Distance=false
    Origin Offset=&Lerp_Vector(A=<0 50 0>, B=<0 -100 -50>, Alpha=$User.JetActiveCurve)
    Override Seed=false
    Velocity Mode=::NewEnumerator1
    Velocity Speed=&MakeCustomFloatFromBool(Bool=&SetBoolByFloatComparison(A=$User.JetActiveCurve, B=0.1), True Float=400, False Float=200)
   )
   &AccelerationForce(Acceleration=&UniformRangedVector(Minimum=<0 -100 0>, Maximum=<0 -250 0>), Coordinate Space=::Local)
   &ApplyInitialForces()
  ]
  particleUpdateScript [
   /SetVariables(Particles.Jitter:NiagaraFloat=&Waveform("Waveform [1]"=::NewEnumerator12, "[1] Frequency"=25, "[1] Amplitude Min/Max"=<0.5 1>)) {isSetParameters=true}
   &ParticleState()
   &Drag(Drag=10, Use Linear Drag=true, Use Rotational Drag=true)
   &SolveForcesAndVelocity(Clamp Velocity=false, Limit Acceleration=false)
   &ScaleSpriteSize(
    Uniform Curve Scale=&Lerp_Float(A=1.33, B=2, Alpha=$System.Timeline.Float)
    Uniform Curve Sprite Scale=DataInterface() {propertyValues {Curve {keys [
        {arriveTangent=1.491, interpMode=RCIM_Cubic, leaveTangent=1.491, tangentMode=RCTM_User, time=0, value=0.5}
        {arriveTangent=0.5, interpMode=RCIM_Cubic, leaveTangent=0.5, time=0.95, value=3}
        <at 1 1 RCIM_Cubic>
       ]}}}
   )
   &DynamicMaterialParameters(
    Index 0 Param 1=&Add_Float(
     A=&RandomRangeFloat(Minimum=0, Maximum=1)
     B=&FloatFromCurve(
      FloatCurve=DataInterface() {propertyValues {Curve {keys [
          <at 0 0 RCIM_Cubic>
          {arriveTangent=1.5947, interpMode=RCIM_Cubic, leaveTangent=1.5947, tangentMode=RCTM_User, time=0.2557, value=0.5257}
          <at 1 1 RCIM_Cubic>
         ]}}}
      CurveIndex=$Particles.Age
      Scale Curve=0.25
     )
    )
    Index 0 Param 3=1
    Index 0 Param 4=1
    Index 2 Param 3=2
    Index 3 Param 1=&Lerp_Float(A=0, B=&Multiply_Float(A=1, B=$Particles.Jitter), Alpha=$System.Timeline.Float)
    Index 3 Param 2=125
    Module.Param0WriteEnabled=true
    Module.Param0WriteEnabled1=true
    Module.Param0WriteEnabled2=true
    Module.Param0WriteEnabled3=true
    Module.Param1WriteEnabled=true
    Module.Param1WriteEnabled1=true
    Module.Param1WriteEnabled2=true
    Module.Param1WriteEnabled3=true
    Module.Param2WriteEnabled=true
    Module.Param2WriteEnabled1=true
    Module.Param2WriteEnabled2=true
    Module.Param2WriteEnabled3=true
    Module.Param3WriteEnabled=true
    Module.Param3WriteEnabled1=true
    Module.Param3WriteEnabled2=true
    Module.Param3WriteEnabled3=true
    Write Parameter Index 1=true
    Write Parameter Index 2=true
    Write Parameter Index 3=true
   ) {module=DynamicMaterialParameters001}
   &CameraOffset(Camera Offset Amount=5)
   &ScaleColor(
    Scale Alpha=&Multiply_Float(
     A=&FloatFromCurve(
      FloatCurve=DataInterface() {propertyValues {Curve {keys [{interpMode=RCIM_Cubic, tangentMode=RCTM_User, time=0, value=0}, <at 0.25 1 RCIM_Cubic>, <at 0.9 0>]}}}
      Scale Curve=&FN_FrontEndFloatModifier(Gameplay=0.8, Front End=0.7)
     )
     B=&FloatFromCurve(
      FloatCurve=DataInterface() {propertyValues {Curve {keys [
          {interpMode=RCIM_Cubic, tangentMode=RCTM_User, time=0, value=0.8}
          {interpMode=RCIM_Cubic, tangentMode=RCTM_User, time=0.5, value=1}
          {tangentMode=RCTM_User, time=1, value=0.8}
         ]}}}
      CurveIndex=Linked() {linkedVariable {name=NPC.FortniteNPC.TimeOfDay_0-1, type=NiagaraFloat}}
     )
    )
    ScaleA=true
    ScaleRGB=true
   )
  ]
  rendererClasses [&NiagaraSpriteRendererProperties]
  renderers [
   {rendererClass=&NiagaraSpriteRendererProperties}
   {rendererClass=&NiagaraSpriteRendererProperties, rendererIndex=1}
  ]
 }
 SmokeGroundRings {
  emitterUpdateScript [
   &EmitterState(
    Inactive Response=::NewEnumerator1
    Life Cycle Mode=::NewEnumerator1
    Loop Behavior=::NewEnumerator1
    Loop Duration Mode=::NewEnumerator1
    UseLoopDelay=false
   )
   &SpawnRate(
    SpawnRate=&Lerp_Float(
     A=0
     B=&FloatFromCurve(
      FloatCurve=DataInterface() {propertyValues {Curve {keys [
          {arriveTangent=-16.5554, interpMode=RCIM_Cubic, leaveTangent=-16.5554, tangentMode=RCTM_User, time=0, value=1}
          {arriveTangent=-0.0405, interpMode=RCIM_Cubic, leaveTangent=-0.0405, tangentMode=RCTM_User, time=0.2, value=0}
         ]}}}
      CurveIndex=$System.NormalizedVelocityRange
      Scale Curve=&Lerp_Float(A=0, B=4, Alpha=$User.JetActiveCurve)
     )
     Alpha=&MakeCustomFloatFromBool(Bool=$User.bIsInWater, True Float=0, False Float=1)
    )
    Use Spawn Probability=false
   )
  ]
  particleSpawnScript [
   &InitializeParticle(
    Color=<rgba 1 0.7225 0.5208 1.5>
    Lifetime=&RandomRangeFloat(Minimum=0.9, Maximum=1.2)
    Mass Mode=::NewEnumerator1
    Sprite Rotation Angle=&UniformRangedFloat(Minimum=-180, Maximum=180)
    Sprite Rotation Mode=::NewEnumerator0
    Sprite Size=&Vector2DFromFloat(Value=600)
    Sprite Size Mode=::NewEnumerator4
    UsePositionOffset=false
    Write Material Random=true
   ) {module=InitializeParticle001}
   /SetVariables(Particles.SpriteFacing=<0 0 1>) {isSetParameters=true}
   &SystemLocation(MaskSpawnGroup=false, Offset=<0 0 1>)
  ]
  particleUpdateScript [
   &ParticleState()
   &DynamicMaterialParameters(
    Index 0 Param 1=&FloatFromCurve(FloatCurve=DataInterface() {propertyValues {Curve {keys [<at 0 0.5>,<at 1 2>]}}})
    Module.Param0WriteEnabled=true
    Module.Param1WriteEnabled=false
    Module.Param2WriteEnabled=false
    Module.Param3WriteEnabled=false
   )
   &ScaleSpriteSize(Scale Factor=&Vector2DFromFloat(Value=&FloatFromCurve(FloatCurve=DataInterface() {propertyValues {Curve {keys [
          {arriveTangent=35.6763, interpMode=RCIM_Cubic, leaveTangent=3.7835, tangentMode=RCTM_User, time=0, value=0}
          {interpMode=RCIM_Cubic, leaveTangent=0.6903, time=1.0011, value=1}
         ]}}})))
   &CameraOffset(Camera Offset Amount=20)
   &ScaleColor(
    Scale Alpha=&FloatFromCurve(
     FloatCurve=DataInterface() {propertyValues {Curve {keys [
         {arriveTangent=0.0112, interpMode=RCIM_Cubic, leaveTangent=0.0112, tangentMode=RCTM_User, time=0, value=0}
         {arriveTangent=0.002, interpMode=RCIM_Cubic, leaveTangent=0.002, time=0.4, value=0.4}
         {interpMode=RCIM_Cubic, leaveTangent=-4.9981, time=1.0011, value=0.002}
        ]}}}
     Scale Curve=&Lerp_Float(A=0, B=$System.TODAlpha, Alpha=$User.JetActiveCurve)
    )
    Scale RGB=&MakeVectorFromLinearColorRGB(Color=$System.TODColor)
    ScaleA=true
    ScaleRGB=true
   )
   &SolveForcesAndVelocity(Clamp Velocity=false, Limit Acceleration=false)
  ]
  rendererClasses [&NiagaraSpriteRendererProperties]
  renderers [{rendererClass=&NiagaraSpriteRendererProperties}]
 }
 Spin {
  emitterUpdateScript [
   &EmitterState(
    Inactive Response=::NewEnumerator1
    Life Cycle Mode=::NewEnumerator1
    Loop Behavior=::NewEnumerator1
    Loop Duration Mode=::NewEnumerator1
    UseLoopDelay=false
   )
   &SpawnBurst_Instantaneous(Spawn Count=4, Use Spawn Probability=false, UseLoopCountLimit=false)
  ]
  particleSpawnScript [
   &InitializeParticle(Color=<rgba 1 0.8926 0.5469 0.33>, Lifetime=5, UsePositionOffset=false, Write Material Random=true)
   &PartitionParticles(ID to Partition:NiagaraInt32=&ReturnExecIndex(), Number Of Partitions=4)
   &InitialMeshOrientation(Mesh Orientation Mode=::NewEnumerator2, Orientation Vector=<0 1 0>, Use Rotation Vector=false)
   &CameraOffset(Camera Offset Amount=&FN_FrontEndFloatModifier(Gameplay=5, Front End=1))
  ]
  particleUpdateScript [
   &ParticleState(Kill Particles When Lifetime Has Elapsed=false, Let Infinitely Lived Particles Die When Emitter Deactivates=true)
   &SkeletalMeshLocation(
    Bone / Socket Index=$Particles.PartitionParticles.Partition
    Bone Sampling Mode=::NewEnumerator4
    MaskSpawnGroup=false
    Sampled Position Offset=&SelectVectorFromArray(Direct Array Index=$Particles.PartitionParticles.Partition, Vector Selection Array=DataInterface() {propertyValues {FloatData []}})
    Skeletal Mesh=$System.Skeletal Mesh
   )
   &UpdateMeshOrientation(Rotation Rate=5, Rotation Vector=<0 1 0>)
   &ScaleColor(
    Scale Alpha=&Add_Float(
     A=&RemapRange(
      Input Value To Remap=&VectorLength(Vector=$Engine.Owner.Velocity)
      Input Min=100
      Input Max=200
      Output Min=0
      Output Max=&Lerp_Float(A=1, B=0, Alpha=$System.Timeline.Float)
     )
     B=&FN_FrontEndFloatModifier(
      Gameplay=0
      Front End=&Multiply_Float(
       A=&Lerp_Float(A=1, B=0, Alpha=$User.JetActiveCurve)
       B=&FloatFromCurve(
        FloatCurve=DataInterface() {propertyValues {Curve {keys [{tangentMode=RCTM_User, time=0.5, value=0}, {tangentMode=RCTM_User, time=1, value=1}]}}}
        CurveIndex=$Engine.System.Age
       )
      )
     )
    )
    ScaleA=true
    ScaleRGB=true
   )
  ]
  rendererClasses [&NiagaraMeshRendererProperties]
  renderers [{rendererClass=&NiagaraMeshRendererProperties, rendererIndex=0}]
 }
 Thrusters {
  emitterUpdateScript [
   &EmitterState(Life Cycle Mode=::NewEnumerator1, Loop Behavior=::NewEnumerator1, Loop Duration Mode=::NewEnumerator1, UseLoopDelay=false)
   &SpawnBurst_Instantaneous(Spawn Count=4, Use Spawn Probability=false, UseLoopCountLimit=false)
  ]
  particleSpawnScript [
   &PartitionParticles(ID to Partition:NiagaraInt32=&ReturnExecIndex(), Number Of Partitions=4)
   &InitializeParticle(
    Color=<rgba 5 5 5 2>
    Mesh Scale Mode=::NewEnumerator4
    Sprite Size Mode=::NewEnumerator3
    Uniform Sprite Size=40
    UsePositionOffset=false
    Write Material Random=true
   )
   &InitialMeshOrientation(
    Mesh Axis To Orient=&SelectVectorFromArray(
     Direct Array Index=$Particles.PartitionParticles.Partition
     Vector Selection Array=DataInterface() {propertyValues {FloatData [<0 0 1>, <0 0 -1>, <0 0 1>, <0 0 -1>]}}
    )
    Mesh Orientation Mode=::NewEnumerator2
    Orientation Coordinate Space=::Local
    Rotation Coordinate Space=::NewEnumerator2
    Use Rotation Vector=true
   )
   &CameraOffset(Camera Offset Amount=2)
  ]
  particleUpdateScript [
   &ParticleState(Kill Particles When Lifetime Has Elapsed=false, Let Infinitely Lived Particles Die When Emitter Deactivates=true)
   &ScaleMeshSize(Scale Factor=&Multiply_VectorByFloat(
     Vector=<1 1 1>
     Float=&MakeCustomFloatFromBool(Bool=&SetBoolByFloatComparison(A=$User.JetActiveCurve, B=0.1), True Float=1, False Float=0)
    ))
   &ScaleColor(Scale Alpha=$User.JetActiveCurve, ScaleA=true, ScaleRGB=true)
   /SetVariables(Particles.Jitter:NiagaraFloat=&Waveform(
     "Waveform [1]"=::NewEnumerator12
     "[1] Input Position"=&RandomRangeFloat(Minimum=0, Maximum=1)
     "[1] Frequency"=25
     "[1] Amplitude Min/Max"=<0.5 1.1>
    )) {isSetParameters=true, module=SetVariables_#3}
   &SkeletalMeshLocation(
    Bone / Socket Index=&SelectIntFromArray(Direct Array Index=$Particles.PartitionParticles.Partition, Int Selection Array=DataInterface() {propertyValues {IntData [0,1,2,3]}})
    Bone Sampling Mode=::NewEnumerator4
    MaskSpawnGroup=false
    Sampled Position Offset=&SelectVectorFromArray(
     Direct Array Index=$Particles.PartitionParticles.Partition
     Vector Selection Array=DataInterface() {propertyValues {FloatData [<4 0 0>, <-4 0 0>, <4 0 0>, <-4 0 0>]}}
    )
    Skeletal Mesh=$System.Skeletal Mesh
    Use Sampled Mesh Attribute as Ribbon ID=false
   )
   &FN_Mesh_Attributes(
    Color=&MakeLinearColorFromVectorAndFloat(
     "Vector (RGB)"=&MakeVectorFromLinearColorRGB(Color=<rgba 2 0.7758 0 1>)
     "Float (Alpha)"=$User.JetActiveCurve
    )
    Create Color Attribute=true
    Create Scale Attribute=true
    Scale=&Multiply_VectorByFloat(
     Vector=&Multiply_VectorByFloat(
      Vector=<0.8 0.8 0.8>
      Float=&MakeCustomFloatFromBool(Bool=&SetBoolByFloatComparison(A=$User.JetActiveCurve, B=0.1), True Float=1, False Float=0)
     )
     Float=$Particles.Jitter
    )
    Write to Module Namespace=true
   )
   &FN_Light_Attributes(
    Color=<rgba 22 10.3657 1.2975 1>
    Create Color Attribute=true
    Create Exponent Attribute=true
    Create Position Attribute=true
    Create Radius Attribute=true
    Enable Particle Light Enable=true
    Exponent=2
    Light Enabled:NiagaraBool=&SetBoolByFloatComparison(A=0.1, B=$User.JetActiveCurve)
    Position=&TransformPosition(
     Position=&AddVectorToPosition(
      Position=$Particles.SkeletalMeshLocation.SampledPosition
      Vector=&SelectVectorFromArray(
       Direct Array Index=$Particles.PartitionParticles.Partition
       Vector Selection Array=DataInterface() {propertyValues {FloatData [<10 0 -30>, <-10 0 -30>, <10 0 -30>, <-10 0 -30>]}}
      )
     )
     Source Space=::Simulation
     Destination Space=::Local
    )
    Radius=&Lerp_Float(A=0, B=40, Alpha=$User.JetActiveCurve)
    Scale Color=&Multiply_Float(
     A=&Multiply_Float(A=$Particles.Jitter, B=$User.JetActiveCurve)
     B=&FN_FrontEndFloatModifier(Gameplay=1, Front End=0)
    )
   )
   &FN_Sprite_Attributes(
    Camera Offset=10
    Color=&MakeLinearColorFromVectorAndFloat(
     "Vector (RGB)"=&MakeVectorFromLinearColorRGB(Color=<rgba 2 0.7758 0 1>)
     "Float (Alpha)"=&Multiply_Float(A=$Particles.Jitter, B=$User.JetActiveCurve)
    )
    Create Camera Offset Attribute=true
    Create Color Attribute=true
    Create Position Attribute=true
    Create SpriteSize Attribute=true
    Position:NiagaraPosition=&AddVectorToPosition(Position=$Particles.Position, Vector=<0 0 -7>)
    SpriteSize=&Multiply_Vector2D_ByFloat(Float=&Multiply_Float(A=$Particles.Jitter, B=$User.JetActiveCurve), Vector2D=<90 50>)
    Write to Module Namespace=true
   )
  ]
  rendererClasses [
   &NiagaraMeshRendererProperties
   &NiagaraLightRendererProperties
   &NiagaraSpriteRendererProperties
  ]
  renderers [
   {rendererClass=&NiagaraMeshRendererProperties, rendererIndex=0}
   {rendererClass=&NiagaraMeshRendererProperties, rendererIndex=1}
   {rendererClass=&NiagaraMeshRendererProperties, rendererIndex=2}
   {rendererClass=&NiagaraLightRendererProperties, rendererIndex=3}
   {rendererClass=&NiagaraSpriteRendererProperties, rendererIndex=4}
  ]
 }
 ThrustersBack {
  emitterUpdateScript [
   &EmitterState(Life Cycle Mode=::NewEnumerator1, Loop Behavior=::NewEnumerator1, Loop Duration Mode=::NewEnumerator1, UseLoopDelay=false)
   &SpawnBurst_Instantaneous(Spawn Count=3, Use Spawn Probability=false, UseLoopCountLimit=false)
  ]
  particleSpawnScript [
   &PartitionParticles(ID to Partition:NiagaraInt32=&ReturnExecIndex(), Number Of Partitions=4)
   &InitializeParticle(
    Color=<rgba 10 10 10 2>
    Mesh Scale=<0.125 0.125 0.125>
    Mesh Scale Mode=::NewEnumerator4
    Sprite Size Mode=::NewEnumerator3
    Uniform Sprite Size=40
    UsePositionOffset=false
    Write Material Random=true
   )
   &InitialMeshOrientation(
    Mesh Axis To Orient=<0 1 0>
    Mesh Orientation Mode=::NewEnumerator2
    Orientation Coordinate Space=::Local
    Orientation Vector=<0 0 1>
    Rotation Coordinate Space=::NewEnumerator2
    Use Rotation Vector=false
   )
   &CameraOffset(Camera Offset Amount=2)
  ]
  particleUpdateScript [
   &ParticleState(Kill Particles When Lifetime Has Elapsed=false, Let Infinitely Lived Particles Die When Emitter Deactivates=true)
   &InterpolateOverTime(
    Rate Of Change=5
    Target Float Value:NiagaraFloat=&RemapRange(
     Input Value To Remap=&VectorLength(Vector=$Engine.Owner.Velocity)
     Input Min=10
     Input Max=20
     Output Min=0
     Output Max=1
    )
   )
   &ScaleMeshSize(Scale Factor=&Multiply_VectorByFloat(
     Vector=&Multiply_Vector(
      A=&Lerp_Vector(A=<1 1 0.75>, B=<1 1 2.5>, Alpha=$Particles.InterpolateOverTime.InterpolateOverTime_Float.Moving Average)
      B=&SelectVectorFromArray(
       Direct Array Index=$Particles.PartitionParticles.Partition
       Vector Selection Array=DataInterface() {propertyValues {FloatData [<1 1 1>, <1 1 1>, <1.5 1.5 1.2>]}}
      )
     )
     Float=&MakeCustomFloatFromBool(Bool=&SetBoolByFloatComparison(A=$User.JetActiveCurve, B=0.1), True Float=1, False Float=0)
    ))
   &ScaleColor(Scale Alpha=$User.JetActiveCurve, ScaleA=true, ScaleRGB=true)
   /SetVariables(Particles.Jitter:NiagaraFloat=&Waveform(
     "Waveform [1]"=::NewEnumerator12
     "[1] Input Position"=&RandomRangeFloat(Minimum=0, Maximum=1)
     "[1] Frequency"=25
     "[1] Amplitude Min/Max"=<0.5 1>
    )) {isSetParameters=true, module=SetVariables_#3}
   &SkeletalMeshLocation(
    Bone / Socket Index=&SelectIntFromArray(Direct Array Index=$Particles.PartitionParticles.Partition, Int Selection Array=DataInterface() {propertyValues {IntData [0,1,2]}})
    MaskSpawnGroup=false
    Mesh Sampling Type=::NewEnumerator5
    Sampled Position Offset=<0 1.6 0>
    Skeletal Mesh=$System.Skeletal Mesh
    Socket Sampling Mode=::NewEnumerator5
    Use Sampled Mesh Attribute as Ribbon ID=false
   )
   &FN_Mesh_Attributes(
    Color=&MakeLinearColorFromVectorAndFloat(
     "Vector (RGB)"=&MakeVectorFromLinearColorRGB(Color=<rgba 2 0.7758 0 1>)
     "Float (Alpha)"=$User.JetActiveCurve
    )
    Create Color Attribute=true
    Create Scale Attribute=true
    Scale=&Multiply_VectorByFloat(
     Vector=&Multiply_VectorByFloat(
      Vector=&SelectVectorFromArray(
       Direct Array Index=$Particles.PartitionParticles.Partition
       Vector Selection Array=DataInterface() {propertyValues {FloatData [<0.1 0.1 0.2>, <0.1 0.1 0.2>, <0.1 0.1 0.15>]}}
      )
      Float=&MakeCustomFloatFromBool(Bool=&SetBoolByFloatComparison(A=$User.JetActiveCurve, B=0.1), True Float=1, False Float=0)
     )
     Float=$Particles.Jitter
    )
    Write to Module Namespace=true
   )
   &FN_Mesh_Attributes(
    Color=&MakeLinearColorFromVectorAndFloat(
     "Vector (RGB)"=&MakeVectorFromLinearColorRGB(Color=<rgba 0 0.7741 5 1>)
     "Float (Alpha)"=&Multiply_Float(A=$Particles.Jitter, B=$User.JetActiveCurve)
    )
    Create Color Attribute=true
    Create Scale Attribute=true
    Scale=&Multiply_VectorByFloat(
     Vector=&SelectVectorFromArray(
      Direct Array Index=$Particles.PartitionParticles.Partition
      Vector Selection Array=DataInterface() {propertyValues {FloatData [<0.3 0.3 1>, <0.3 0.3 1>, <0.6 0.6 1>]}}
     )
     Float=&MakeCustomFloatFromBool(Bool=&SetBoolByFloatComparison(A=$User.JetActiveCurve, B=0.1), True Float=1, False Float=0)
    )
    Write to Module Namespace=true
   ) {module=FN_Mesh_Attributes001}
   &FN_Sprite_Attributes(
    Camera Offset=5
    Color=&MakeLinearColorFromVectorAndFloat(
     "Vector (RGB)"=&MakeVectorFromLinearColorRGB(Color=<rgba 0 1.3804 5 1>)
     "Float (Alpha)"=&Multiply_Float(A=$Particles.Jitter, B=$User.JetActiveCurve)
    )
    Create Camera Offset Attribute=true
    Create Color Attribute=true
    Create Position Attribute=true
    Create SpriteSize Attribute=true
    Position:NiagaraPosition=&AddVectorToPosition(Position=$Particles.Position, Vector=<0 -2 0>)
    SpriteSize=&Multiply_Vector2D_ByFloat(
     Float=&Multiply_Float(A=$Particles.Jitter, B=$User.JetActiveCurve)
     Vector2D=&Multiply_Vector2D_ByFloat(
      Float=&SelectFloatFromArray(Direct Array Index=$Particles.PartitionParticles.Partition, Float Selection Array=DataInterface() {propertyValues {FloatData [1,1,1.3]}})
      Vector2D=<35 35>
     )
    )
    Write to Module Namespace=true
   )
  ]
  rendererClasses [&NiagaraMeshRendererProperties, &NiagaraSpriteRendererProperties]
  renderers [
   {rendererClass=&NiagaraMeshRendererProperties, rendererIndex=0}
   {rendererClass=&NiagaraMeshRendererProperties, rendererIndex=1}
   {rendererClass=&NiagaraSpriteRendererProperties, rendererIndex=2}
   {rendererClass=&NiagaraMeshRendererProperties, rendererIndex=3}
  ]
 }
 WindstreaksJet {
  emitterUpdateScript [
   &EmitterState()
   &SpawnRate(Spawn Probability=$User.JetActiveCurve, SpawnRate=20, Use Spawn Probability=true)
  ]
  particleSpawnScript [
   &InitializeParticle(
    Color Maximum=<rgba 0.4 0.2465 0.1693 0.3>
    Color Minimum=<rgba 0.4 0.3257 0.2062 0.35>
    Color Mode=::NewEnumerator2
    Lifetime=&UniformRangedFloat(Minimum=0.3, Maximum=0.5)
    Sprite Size=&UniformRangedVector2D(Minimum=<30 40>, Maximum=<40 50>)
    Sprite Size Mode=::NewEnumerator4
    UsePositionOffset=false
    Write Material Random=true
   )
   &SystemLocation(
    MaskSpawnGroup=false
    Offset=&SelectVectorFromArray(
     Array Sampling Mode=::NewEnumerator0
     Direct Array Index=?
     Vector Selection Array=DataInterface() {propertyValues {FloatData [<50 -50 0>, <-50 -50 0>, <50 50 0>, <-50 50 0>]}}
     Random Seed=0
    )
   )
   &ShapeLocation(
    Cylinder Height=1
    Cylinder Radius=20
    MaskSpawnGroup=false
    Override Seed=false
    Shape Primitive=::NewEnumerator1
    Surface Only=false
    UseOwnerScale=true
   )
   &AddVelocity(Module.Use Falloff Distance=false, Origin Offset=<0 0 -50>, Override Seed=false, Velocity Mode=::NewEnumerator1, Velocity Speed=250)
   &ApplyInitialForces()
   &SubUVAnimation(
    Evaluation Type=::NewEnumerator0
    "Sprite Renderer "=DataInterface() {propertyValues {SpriteRenderer=&self:WindstreaksJet_0.NiagaraSpriteRendererProperties_0}}
    SubUV Animation Mode:ENiagara_SubUVLookupModeV2=::NewEnumerator2
    UseEndFrame=false
    UseStartFrame=false
   )
  ]
  particleUpdateScript [
   &ParticleState()
   &ScaleSpriteSize(Scale Factor=&Vector2DFromCurve(
     Vector2Curve=DataInterface() {propertyValues {
       XCurve {keys [
         {arriveTangent=2.5428, interpMode=RCIM_Cubic, leaveTangent=2.5428, tangentMode=RCTM_User, time=0, value=1}
         {arriveTangent=0.1281, interpMode=RCIM_Cubic, leaveTangent=0.1281, tangentMode=RCTM_User, time=1, value=2}
        ]}
       YCurve {keys [
         {arriveTangent=2.8823, interpMode=RCIM_Cubic, leaveTangent=2.8823, tangentMode=RCTM_User, time=0, value=1}
         {arriveTangent=2.0297, interpMode=RCIM_Cubic, leaveTangent=2.0297, tangentMode=RCTM_User, time=1, value=2}
        ]}
      }}
     ScaleCurve=&Multiply_Vector2D_ByFloat(
      Float=&RandomRangeFloat(Minimum=1, Maximum=2)
      Vector2D=&RandomRangeVector2D(Minimum=<0.6 0.75>, Maximum=<0.8 1>)
     )
    ))
   &DynamicMaterialParameters(
    Index 0 Param 1=20
    Index 0 Param 2=1
    Index 0 Param 3=1
    Index 0 Param 4=&FloatFromCurve(FloatCurve=DataInterface() {propertyValues {Curve {keys [{tangentMode=RCTM_User, time=0.1, value=0}, {tangentMode=RCTM_User, time=1, value=0.5}]}}})
    Module.Param0WriteEnabled=true
    Module.Param1WriteEnabled=true
    Module.Param2WriteEnabled=true
    Module.Param3WriteEnabled=true
   ) {module=DynamicMaterialParameters001}
   &Drag(Drag=5, Use Linear Drag=true, Use Rotational Drag=true)
   &InheritSourceMovement(
    Applied Position Scale=&Multiply_VectorByFloat(
     Vector=<1 1 1>
     Float=&FloatFromCurve(FloatCurve=DataInterface() {propertyValues {Curve {keys [<at 0.5 1>,<at 1 0>]}}})
    )
    ApplyForce=false
    ApplyPosition=true
    ApplyPositionFalloff=false
   )
   &SolveForcesAndVelocity(Clamp Velocity=false, Limit Acceleration=false)
   &ScaleColor(
    Scale Alpha=&FloatFromCurve(
     FloatCurve=DataInterface() {propertyValues {Curve {keys [
         {arriveTangent=9.0725, interpMode=RCIM_Cubic, leaveTangent=9.0725, tangentMode=RCTM_User, time=0, value=0}
         {arriveTangent=-2.5249, interpMode=RCIM_Cubic, leaveTangent=-2.5249, tangentMode=RCTM_User, time=0.1408, value=0.9847}
         {arriveTangent=-0.0405, interpMode=RCIM_Cubic, leaveTangent=-0.0405, tangentMode=RCTM_User, time=1, value=0}
        ]}}}
     Scale Curve=&Multiply_Float(
      A=$System.TODAlpha
      B=&SelectFloatFromArray(Direct Array Index=$Engine.QualityLevel, Float Selection Array=DataInterface() {propertyValues {FloatData [0.5,0.5,1,1,1]}})
     )
    )
    Scale RGB=&MakeVectorFromLinearColorRGB(Color=$System.TODColor)
    ScaleA=true
    ScaleRGB=true
   )
  ]
  rendererClasses [&NiagaraSpriteRendererProperties]
  renderers [{rendererClass=&NiagaraSpriteRendererProperties}]
 }
}
summary {
 systemName=NS_Clayplug_Emote_Jetwheels
 emitters [
  {
   emitterName=Thrusters
   rendererClasses [
    &NiagaraMeshRendererProperties
    &NiagaraLightRendererProperties
    &NiagaraSpriteRendererProperties
   ]
  }
  {emitterName=WindstreaksJet, rendererClasses [&NiagaraSpriteRendererProperties]}
  {emitterName=Chunks, rendererClasses [&NiagaraSpriteRendererProperties]}
  {emitterName=Default_Emote_TraversalSmoke_Loop001, rendererClasses [&NiagaraSpriteRendererProperties]}
  {emitterName=SmokeGroundRings, rendererClasses [&NiagaraSpriteRendererProperties]}
  {emitterName=Spin, rendererClasses [&NiagaraMeshRendererProperties]}
  {emitterName=ThrustersBack, rendererClasses [&NiagaraMeshRendererProperties, &NiagaraSpriteRendererProperties]}
 ]
 userVariables [
  User.bIsFrontEnd {type=NiagaraBool}
  User.bIsFrontEndPreview {type=NiagaraBool}
  User.bIsInWater {type=NiagaraBool}
  User.DustSpawnCurve {type=NiagaraFloat, defaultValue=1, description=""}
  User.GlobalScale {type=NiagaraFloat, defaultValue=1, description=""}
  User.JetActiveCurve {type=NiagaraFloat, defaultValue=0, description=""}
 ]
}
