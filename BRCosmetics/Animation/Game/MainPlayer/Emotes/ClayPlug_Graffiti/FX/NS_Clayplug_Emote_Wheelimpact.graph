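# Niagara system for the wheel impact effect in the ClayPlug Graffiti emote.
# Three burst emitters throw gravity-pulled debris chunks, an expanding upward-facing smoke ground ring, and short-lived windstreak sprites that grow and fade.
# All three set spawn probability to zero when User.bIsInWater is true, suppressing the effect in water.
# Ring and streaks take their tint from System.TODColor and animate dynamic material parameters, while chunks use random curve-sampled colors.
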NiagaraSystem /BRCosmetics/Animation/Game/MainPlayer/Emotes/ClayPlug_Graffiti/FX/NS_Clayplug_Emote_Wheelimpact
dependencies {
 usedDataInterfaces [
  &NiagaraDataInterfaceColorCurve
  &NiagaraDataInterfaceCurve
  &NiagaraDataInterfaceSpriteRendererInfo
  &NiagaraDataInterfaceVector2DCurve
  &NiagaraDataInterfaceSkeletalMesh
 ]
 usedDynamicInputs [
  &MakeFloatFromVector2d
  &Multiply_Vector2
  &RandomRangeVector
 ]
 usedModules [
  &self:SystemSpawnScript.SystemScriptSource.SystemScriptGraph.NiagaraNodeAssignment_0.SetVariables_#1
  &SystemState
  /BRCosmetics/Characters/Player/Female/Medium/Bodies/F_MED_PiperShelf_Pearl/FX/TimeOfDayGlobal
 ]
}
emitters {
 Chunks {
  emitterUpdateScript [
   &EmitterState()
   &SpawnBurst_Instantaneous(
    Spawn Count=&RandomRangeInt(Minimum=15, Maximum=20)
    Spawn Probability=&MakeCustomFloatFromBool(Bool=$User.bIsInWater, True Float=0, False Float=1)
    Use Spawn Probability=true
    UseLoopCountLimit=false
   )
  ]
  particleSpawnScript [
   &PartitionParticles(Number Of Partitions=4)
   &InitializeParticle(
    Color=&ColorFromCurve(
     ColorCurve=DataInterface() {propertyValues {
       AlphaCurve {keys [<at -0.0023 1 RCIM_Cubic>]}
       BlueCurve {keys [<at -0.0023 0.2122 RCIM_Cubic>,<at 0.1383 0.2122>,<at 0.2178 0.1859>,<at 0.4486 0.75>,<at 0.7126 0.1859>,<at 1.0234 0.8802>]}
       GreenCurve {keys [<at -0.0023 0.2325 RCIM_Cubic>,<at 0.1383 0.2325>,<at 0.2178 0.2005>,<at 0.4486 0.8456>,<at 0.7126 0.2122>,<at 1.0234 0.9474>]}
       RedCurve {keys [<at -0.0023 0.25 RCIM_Cubic>,<at 0.1383 0.25>,<at 0.2178 0.2188>,<at 0.4486 1>,<at 0.7126 0.2188>,<at 1.0234 1>]}
      }}
     CurveIndex=&RandomRangeFloat(Minimum=0, Maximum=1)
    )
    Lifetime Max=1
    Lifetime Min=0.25
    Lifetime Mode=::NewEnumerator1
    Sprite Rotation Mode=::NewEnumerator2
    Sprite Size Mode=::NewEnumerator1
    Uniform Sprite Size Max=8
    Uniform Sprite Size Min=6
    UsePositionOffset=false
    Write Material Random=true
   ) {module=InitializeParticle001}
   &ShapeLocation(MaskSpawnGroup=false, Override Seed=false, Shape Primitive=::NewEnumerator1, Surface Only=false, UseOwnerScale=true)
   &AddVelocity(
    Module.Use Falloff Distance=false
    Origin Offset=<0 0 -50>
    Override Seed=false
    Velocity Mode=::NewEnumerator1
    Velocity Speed=&RandomRangeFloat(Minimum=150, Maximum=300)
   )
  ]
  particleUpdateScript [
   &ParticleState()
   &Drag(Use Linear Drag=true, Use Rotational Drag=true)
   &GravityForce(Gravity=<0 0 -500>)
   &SolveForcesAndVelocity(Clamp Velocity=false, Limit Acceleration=false)
   &ScaleColor(Scale RGB=<1.2 1.2 1.2>, ScaleA=true, ScaleRGB=true)
   &ScaleSpriteSize(Uniform Curve Sprite Scale=DataInterface() {propertyValues {Curve {keys [
        {tangentMode=RCTM_User, time=0, value=1}
        {arriveTangent=-0.0405, interpMode=RCIM_Cubic, leaveTangent=-0.0405, tangentMode=RCTM_User, time=0.7009, value=1}
        {arriveTangent=-8.0089, interpMode=RCIM_Cubic, leaveTangent=-8.0089, tangentMode=RCTM_User, time=1, value=0}
       ]}}})
   &SpriteRotationRate()
   &ScaleSpriteSizeBySpeed(
    Max Scale Factor=&Vector2DFromFloat(Value=1)
    Min Scale Factor=&Vector2DFromFloat(Value=0.25)
    Velocity Threshold=200
   )
  ]
  rendererClasses [&NiagaraSpriteRendererProperties]
  renderers [{rendererClass=&NiagaraSpriteRendererProperties}]
 }
 SmokeGroundRings {
  emitterUpdateScript [
   &EmitterState()
   &SpawnBurst_Instantaneous(
    Spawn Probability=&MakeCustomFloatFromBool(Bool=$User.bIsInWater, True Float=0, False Float=1)
    Use Spawn Probability=true
    UseLoopCountLimit=false
   )
  ]
  particleSpawnScript [
   &InitializeParticle(
    Lifetime=0.25
    Mass Mode=::NewEnumerator1
    Sprite Rotation Mode=::NewEnumerator1
    Sprite Size=<650 650>
    Sprite Size Mode=::NewEnumerator4
    UsePositionOffset=false
    Write Material Random=true
   ) {module=InitializeParticle001}
   /SetVariables(Particles.SpriteFacing=<0 0 1>) {isSetParameters=true}
   &SystemLocation(MaskSpawnGroup=false, Offset=<0 0 0.5>)
  ]
  particleUpdateScript [
   &ParticleState()
   &DynamicMaterialParameters(
    Index 0 Param 1=&FloatFromCurve(FloatCurve=DataInterface() {propertyValues {Curve {keys [<at 0 1.2>,<at 1 1.65>]}}})
    Module.Param0WriteEnabled=true
    Module.Param1WriteEnabled=false
    Module.Param2WriteEnabled=false
    Module.Param3WriteEnabled=false
   )
   &ScaleSpriteSize(Scale Factor=&Vector2DFromFloat(Value=&FloatFromCurve(FloatCurve=DataInterface() {propertyValues {Curve {keys [
          {arriveTangent=1.516, interpMode=RCIM_Cubic, leaveTangent=1.516, tangentMode=RCTM_User, time=0, value=0.45}
          {interpMode=RCIM_Cubic, leaveTangent=0.6903, time=1.0011, value=1}
         ]}}})))
   &CameraOffset(Camera Offset Amount=20)
   &ScaleColor(
    Scale Alpha=&FloatFromCurve(FloatCurve=DataInterface() {propertyValues {Curve {keys [
         {arriveTangent=0.0112, interpMode=RCIM_Cubic, leaveTangent=0.0112, tangentMode=RCTM_User, time=0, value=0}
         {arriveTangent=0.002, interpMode=RCIM_Cubic, leaveTangent=0.002, time=0.2, value=0.3875}
         {interpMode=RCIM_Cubic, leaveTangent=-4.9981, time=1.0011, value=0.002}
        ]}}})
    Scale RGB=&MakeVectorFromLinearColorRGB(Color=$System.TODColor)
    ScaleA=true
    ScaleRGB=true
   )
   &SolveForcesAndVelocity(Clamp Velocity=false, Limit Acceleration=false)
  ]
  rendererClasses [&NiagaraSpriteRendererProperties]
  renderers [{rendererClass=&NiagaraSpriteRendererProperties}]
 }
 WindstreaksJet {
  emitterUpdateScript [
   &EmitterState()
   &SpawnBurst_Instantaneous(
    Spawn Count=16
    Spawn Probability=&MakeCustomFloatFromBool(Bool=$User.bIsInWater, True Float=0, False Float=1)
    Use Spawn Probability=true
    UseLoopCountLimit=false
   )
  ]
  particleSpawnScript [
   &InitializeParticle(
    Color Maximum=<rgba 0.3 0.2901 0.2406 0.3>
    Color Minimum=<rgba 0.35 0.3095 0.2443 0.35>
    Color Mode=::NewEnumerator2
    Lifetime=&UniformRangedFloat(Minimum=0.2, Maximum=0.3)
    Sprite Size=&UniformRangedVector2D(Minimum=<15 45>, Maximum=<35 70>)
    Sprite Size Mode=::NewEnumerator4
    UsePositionOffset=false
    Write Material Random=true
   )
   &ShapeLocation(
    Cylinder Height=1
    Enable Lathe Profile=true
    MaskSpawnGroup=false
    Override Seed=false
    Shape Primitive=::NewEnumerator1
    Surface Only=true
    Surface Only Band Thickness=1
    Use Endcaps In Surface Only Mode=false
    UseOwnerScale=true
   )
   &AddVelocity(Module.Use Falloff Distance=false, Origin Offset=<0 0 -15>, Override Seed=false, Velocity Mode=::NewEnumerator1, Velocity Speed=250)
   &ApplyInitialForces()
   &SubUVAnimation(
    Evaluation Type=::NewEnumerator0
    "Sprite Renderer "=DataInterface() {propertyValues {SpriteRenderer=&self:WindstreaksJet_0.NiagaraSpriteRendererProperties_0}}
    SubUV Animation Mode:ENiagara_SubUVLookupModeV2=::NewEnumerator2
    UseEndFrame=false
    UseStartFrame=false
   )
  ]
  particleUpdateScript [
   &ParticleState()
   &ScaleSpriteSize(Scale Factor=&Vector2DFromCurve(
     Vector2Curve=DataInterface() {propertyValues {
       XCurve {keys [
         {arriveTangent=2.5428, interpMode=RCIM_Cubic, leaveTangent=2.5428, tangentMode=RCTM_User, time=0, value=1}
         {arriveTangent=0.1281, interpMode=RCIM_Cubic, leaveTangent=0.1281, tangentMode=RCTM_User, time=1, value=2}
        ]}
       YCurve {keys [
         {arriveTangent=2.8823, interpMode=RCIM_Cubic, leaveTangent=2.8823, tangentMode=RCTM_User, time=0, value=1}
         {arriveTangent=2.0297, interpMode=RCIM_Cubic, leaveTangent=2.0297, tangentMode=RCTM_User, time=1, value=2}
        ]}
      }}
     ScaleCurve=&Multiply_Vector2D_ByFloat(
      Float=&RandomRangeFloat(Minimum=1, Maximum=2)
      Vector2D=&RandomRangeVector2D(Minimum=<0.6 0.75>, Maximum=<0.8 1>)
     )
    ))
   &DynamicMaterialParameters(
    Index 0 Param 1=20
    Index 0 Param 2=1
    Index 0 Param 3=1
    Index 0 Param 4=&FloatFromCurve(FloatCurve=DataInterface() {propertyValues {Curve {keys [{tangentMode=RCTM_User, time=0.1, value=0}, {tangentMode=RCTM_User, time=1, value=0.5}]}}})
    Module.Param0WriteEnabled=true
    Module.Param1WriteEnabled=true
    Module.Param2WriteEnabled=true
    Module.Param3WriteEnabled=true
   ) {module=DynamicMaterialParameters001}
   &Drag(Drag=3, Use Linear Drag=true, Use Rotational Drag=true)
   &InheritSourceMovement(
    Applied Position Scale=&Multiply_VectorByFloat(
     Vector=<1 1 1>
     Float=&FloatFromCurve(FloatCurve=DataInterface() {propertyValues {Curve {keys [<at 0.5 1>,<at 1 0>]}}})
    )
    ApplyForce=false
    ApplyPosition=true
    ApplyPositionFalloff=false
   )
   &SolveForcesAndVelocity(Clamp Velocity=false, Limit Acceleration=false)
   &ScaleColor(
    Scale Alpha=&FloatFromCurve(FloatCurve=DataInterface() {propertyValues {Curve {keys [
         {arriveTangent=9.0725, interpMode=RCIM_Cubic, leaveTangent=9.0725, tangentMode=RCTM_User, time=0, value=0}
         {arriveTangent=-2.5249, interpMode=RCIM_Cubic, leaveTangent=-2.5249, tangentMode=RCTM_User, time=0.1408, value=0.9847}
         {arriveTangent=-0.0405, interpMode=RCIM_Cubic, leaveTangent=-0.0405, tangentMode=RCTM_User, time=1, value=0}
        ]}}})
    Scale RGB=&MakeVectorFromLinearColorRGB(Color=$System.TODColor)
    ScaleA=true
    ScaleRGB=true
   )
  ]
  rendererClasses [&NiagaraSpriteRendererProperties]
  renderers [{rendererClass=&NiagaraSpriteRendererProperties}]
 }
}
summary {
 systemName=NS_Clayplug_Emote_Wheelimpact
 emitters [
  {emitterName=Chunks, rendererClasses [&NiagaraSpriteRendererProperties]}
  {emitterName=WindstreaksJet, rendererClasses [&NiagaraSpriteRendererProperties]}
  {emitterName=SmokeGroundRings, rendererClasses [&NiagaraSpriteRendererProperties]}
 ]
 userVariables [
  User.bIsFrontEnd {type=NiagaraBool}
  User.bIsFrontEndPreview {type=NiagaraBool}
  User.bIsInWater {type=NiagaraBool}
 ]
}
